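; FZ variant of the SilverBlazer shield bubble, a deployed zone granting players inside damage immunity, targeting blocks and optional weapon holstering or low gravity.
; A server timer polls overlapping players rather than relying on overlap events, adding or removing cues and gameplay effects via per-pawn handles.
; It optionally destroys buildings within its radius at spawn.
; Timelines drive deploy, warning and death visuals; the deployed bubble registers as an AI occluder.
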(event EventBeginPlay
  ;[0.05345 0.15 0.02229 0.5 36] Begin Play
  (|CallParentFunction)
  (Utilities|FlowControl|Sequence
    (:then_0
      ;[1 1 1] Idle FX Start
      (bind _returnvalue (Niagara|SpawnSystematLocation :SystemTemplate "/ShieldBubble/Assets/FX/FZ/NS_ShieldBubble_FZ_IdleSwirl.NS_ShieldBubble_FZ_IdleSwirl" :Location (Transformation|GetActorLocation) :Scale "1.000000,1.000000,1.000000" :bAutoDestroy false :bAutoActivate true :PoolingMethod "None" :bPreCullCheck true))
      (Variables|Default|SetIdleSparkleFX _returnvalue))
    (:then_1
      (Transformation|SetWorldScale3D (Variables|Default|GetCollisionSphere) (Math|Conversions|ToVector(Float) (ScalableFloat|GetValueatLevel (Variables|SilverBronco|GetMaxScale) (Variables|Default|GetAbilityLevel))))
      (Default|SetupOverlappingGameplayEffectstoApply)
      (Utilities|FlowControl|DelayUntilNextTick)
      (Collision|SetCollisionEnabled (Variables|Default|GetCollisionSphere) "QueryOnly") ; wait until the bubble is fully deployed to check if there are players within it
      (Networking|SwitchHasAuthority
        (:Authority
          (Utilities|Time|SetTimerbyFunctionName 0 "CheckOverlappingActors" (ScalableFloat|GetValueatLevel (Variables|Default|GetOverlapCheckFrequency)) true)
          (Default|TryBlockBuilding))))))

(event OnComponentEndOverlap(CollisionSphere)
  (params (OverlappedComponent :type Primitive%20Component%20Object%20Reference) (OtherActor :type Actor%20Object%20Reference) (OtherComp :type Primitive%20Component%20Object%20Reference) (OtherBodyIndex :type Integer))
  ;[1 1 1 1 36] 1. Exit Bubble
  (Networking|SwitchHasAuthority
    (:Authority
      (Utilities|FlowControl|Delay 0.0)
      (bind _asactor (Utilities|Casting|CastToActor (Variables|Default|GetRunInLensEffectInterface)))
      (Actor|DestroyActor _asactor))))

(event OnComponentBeginOverlap(CollisionSphere)
  (params (OverlappedComponent :type Primitive%20Component%20Object%20Reference) (OtherActor :type Actor%20Object%20Reference) (OtherComp :type Primitive%20Component%20Object%20Reference) (OtherBodyIndex :type Integer) (bFromSweep :type Boolean) (SweepResult :type Hit%20Result%20Structure%20%28by%20ref%29))
  ;[1 1 1 1 36] 1. Enter Bubble
  (Networking|SwitchHasAuthority
    (:Authority
      (Utilities|FlowControl|Delay 0.0)
      (bind _asactor (Utilities|Casting|CastToActor (Variables|Default|GetRunOutLensEffectInterface)))
      (Actor|DestroyActor _asactor))))

(event Custom|Add%20Player%20To%20Array
  (params (Player :type Player%20Pawn%20Athena%20Object%20Reference))
  ;[1 1 1] 2. Add to Array
  (if (not (Utilities|Array|ContainsItem (Variables|Default|GetPlayersInZone) Player))
    (Utilities|Array|AddUnique (Variables|Default|GetPlayersInZone) Player)
    (CallFunction|ApplyEffects Player)
    (CallFunction|NotifyPlayerEnteredZone Player)))

(event Custom|Remove%20Player%20From%20Array
  (params (Player :type Player%20Pawn%20Athena%20Object%20Reference))
  ;[1 1 1] 2. Remove from Array
  (Utilities|Array|RemoveItem (Variables|Default|GetPlayersInZone) Player)
  (CallFunction|RemoveEffects Player)
  (CallFunction|NotifyPlayerLeftZone Player))

(event Custom|Apply%20Effects
  (params (Player :type Player%20Pawn%20Athena%20Object%20Reference))
  ;[1 1 1] 3. Apply Effects
  (GameplayCue|AddGameplayCue Player (Variables|Default|GetGameplay%20Cue%20Tag))
  (Default|ApplyOverlappingEffectstoPawn Player))

(event Custom|Remove%20Effects
  (params (Player :type Player%20Pawn%20Athena%20Object%20Reference))
  ;[1 1 1] 3. Remove Effects
  (GameplayCue|RemoveGameplayCue Player (Variables|Default|GetTag_GCNL_Immune))
  (GameplayCue|RemoveGameplayCue Player (Variables|Default|GetGameplay%20Cue%20Tag))
  (Default|RemoveOverlappingEffectsfromPawn Player))

(event EventEventScaleSafeZone
  (bind (_origin _boxextent _sphereradius) (Collision|GetComponentBounds (Variables|Default|GetSafeZone_IN)))
  (bind _returnvalue (Class|AthenaAIServiceZoneManager|GetAthenaAIServiceZoneManager true))
  ;[0.4503 0.7522 1] Scale Safe Zone & Deploy FX
  (Utilities|FlowControl|Sequence
    (:then_0
      (|CallParentFunction))
    (:then_1
      (Transformation|SetWorldScale3D (Variables|Default|GetSafeZone_IN) "0, 0, 0")
      (Components|Timeline|SetPlayRate (Variables|BGA_Athena_SilverBlazer_FZ|GetScaleSafezone_IN) (/ 1.0 (ScalableFloat|GetValueatLevel (Variables|SilverBronco|GetScaleDuration))))
      (Utilities|Struct|SetmembersinSilverBlazerScaleSafeZoneUpdateData (Variables|Default|GetScaleSafeZoneInUpdateData) "Deploy" "FresnelBrightness" (Variables|Default|GetSafeZone_IN) (ScalableFloat|GetValueatLevel (Variables|SilverBronco|GetMaxScale)) 5e+01 1.0 5.0)
      (bind (_direction _scalesafezone _deployscanin _deploydoneemis) (Timeline|ScaleSafeZone_IN Play)
        (:Update
          (bind _structout (Utilities|Struct|SetmembersinSilverBlazerScaleSafeZoneUpdateData (Variables|Default|GetScaleSafeZoneInUpdateData) _scalesafezone _deployscanin _deploydoneemis))
          (CallFunction|NativeScaleSafeZoneUpdate _structout))
        (:Finished
          (Utilities|FlowControl|Sequence
            (:then_0
              (Utilities|IsValid _returnvalue
                (:Is%20Valid
                  (Class|AthenaAIServiceZoneManager|RegisterOccluder _returnvalue self _sphereradius))))))))))

(event Custom|Error%20Check)

(event EventDeathCosmetics
  (bind _safezone_in (Variables|Default|GetSafeZone_IN))
  (bind _safezone_in_2 (Variables|Default|GetSafeZone_IN))
  ;[1 1 1] Death Cosmetics
  (Utilities|FlowControl|Sequence
    (:then_0
      (|CallParentFunction))
    (:then_1
      (Rendering|Material|SetScalarParameterValueonMaterials _safezone_in "FresnelBrightness" 1.0)
      (Rendering|Material|SetScalarParameterValueonMaterials _safezone_in "IntBrightness" 1.5)
      (bind (_direction _fadeout _shrinkwpo _opacity) (Timeline|Timeline_Death_IN PlayFromStart)
        (:Update
          (Rendering|Material|SetScalarParameterValueonMaterials _safezone_in_2 "DissolveAmount" _fadeout)
          (Rendering|Material|SetScalarParameterValueonMaterials _safezone_in_2 "Shrink" _shrinkwpo)
          (Rendering|Material|SetScalarParameterValueonMaterials _safezone_in_2 "OverallOpacity" _opacity)
          ;[1 1 1] Kill Idle FX
          (Components|Activation|SetActive (Variables|Default|GetIdleSparkleFX) false false))))))

(event EventEndPlay
  (params (EndPlayReason :type EEndPlayReason%20Enum :default "Destroyed"))
  ;[1 1 1 1 36] End Play
  (for _array_element (Variables|Default|GetPlayersInZone)
    (CallFunction|RemovePlayerfromArray _array_element)))

(event EventPlayWarningEffectBeforeShieldGoAway
  (bind _safezone_in (Variables|Default|GetSafeZone_IN))
  ;[1 0.5308 0.5507] WarningFX Before Shield Go away
  (|CallParentFunction)
  (Rendering|Material|SetVectorParameterValueonMaterials (Variables|Default|GetSafeZone_IN) "ColorCenter" "0.027000,0.008388,0.338542")
  (bind (_direction _warningemissive) (Timeline|Timeline PlayFromStart)
    (:Update
      (Rendering|Material|SetScalarParameterValueonMaterials _safezone_in "FresnelBrightness" (Math|Float|Lerp 1.0 12.0 _warningemissive))
      (Rendering|Material|SetScalarParameterValueonMaterials _safezone_in "IntBrightness" (Math|Float|Lerp 1.5 12.0 _warningemissive)))))

(event Custom|OnPawnExitCollisionIgnoredVehicle
  (params (Vehicle :type Fort%20Vehicle%20Interface%20Interface%20%28by%20ref%29) (PlayerPawn :type Fort%20Pawn%20Object%20Reference) (SeatIndex :type Integer))
  ;[1 1 1] Remove the vehicle exiting player from ignored actors just ot make sure collision checks will work
  (Collision|IgnoreActorwhenMoving (Variables|Default|GetSafeZone) PlayerPawn false))

(event EventRunIn
  (params (Actor :type Actor%20Object%20Reference)))

(event Custom|NotifyPlayerEnteredZone
  (params (Actor :type Actor%20Object%20Reference))
  ; Leaving RunIn and RunOut empty since we want to override the parent function for this blueprint and handle it inside the new overlap logic.
  (Networking|CanExecuteCosmeticEvents
    (:True
      (|CallParentFunction Actor))))

(event EventRunOut
  (params (Actor :type Actor%20Object%20Reference)))

(event Custom|NotifyPlayerLeftZone
  (params (Actor :type Actor%20Object%20Reference))
  ; Leaving RunIn and RunOut empty since we want to override the parent function for this blueprint and handle it inside the new overlap logic.
  (Networking|CanExecuteCosmeticEvents
    (:True
      (|CallParentFunction Actor))))

(fn ApplyOverlappingEffectsToPawn
  (params (OverlappingPawn :type Fort%20Pawn%20Object%20Reference))
  (Utilities|IsValid OverlappingPawn
    (:Is%20Valid
      (bind _output_get (LocalVariables|SetASC (Ability|GetFortAbilitySystemComponent OverlappingPawn)))
      (Utilities|IsValid _output_get
        (:Is%20Valid
          (LocalVariables|SetSelfASC (Ability|GetFortAbilitySystemComponent self))
          (Default|RemoveOverlappingEffectsfromPawn (Variables|GetOverlappingPawn))
          (for _array_element (Variables|Default|GetOverlappingGameplayEffectsToApply)
            (if (Utilities|IsValidClass _array_element)
              (bind _returnvalue (GameplayEffects|FortApplyGameplayEffectToTarget :self (Ability|GetFortAbilitySystemComponent self) :GameplayEffectClass _array_element :TargetOfEffect (LocalVariables|GetASC) :OptionalSourceObject self :OptionalEffectCauser self :Level 0.0))
              (Utilities|Array|Add (LocalVariables|GetEffectHandles) _returnvalue)))
          (Utilities|Map|Add (Variables|Default|GetActorToGameplayEffectHandles) (Variables|GetOverlappingPawn) (Utilities|Struct|MakeBPSFZGameplayEffectHandles (LocalVariables|GetEffectHandles))))))))

(fn CheckOverlappingActors
  (Utilities|Array|Clear (LocalVariables|GetCurrentOverlappingPlayers))
  (for _array_element (Collision|GetOverlappingActors (Variables|Default|GetCollisionSphere) "/Game/Athena/PlayerPawn_Athena.PlayerPawn_Athena_C" :!class /Script/Engine.PrimitiveComponent)
    (bind _asplayer_pawn_athena (Utilities|Casting|CastToPlayerPawn_Athena _array_element))
    (Utilities|Array|AddUnique (LocalVariables|GetCurrentOverlappingPlayers) _asplayer_pawn_athena)
    (CallFunction|AddPlayertoArray _asplayer_pawn_athena))
  ;[1 1 1] Clear not overlapping players
  (for _array_element_2 (Variables|Default|GetPlayersInZone)
    (if (not (Utilities|Array|ContainsItem (LocalVariables|GetCurrentOverlappingPlayers) _array_element_2))
      (CallFunction|RemovePlayerfromArray _array_element_2))))

(fn RemoveOverlappingEffectsFromPawn
  (params (OverlappingPawn :type Fort%20Pawn%20Object%20Reference))
  (bind _overlappingpawn (Variables|GetOverlappingPawn))
  (bind _actortogameplayeffecthandles (Variables|Default|GetActorToGameplayEffectHandles))
  (bind (_value _returnvalue) (Utilities|Map|Find _actortogameplayeffecthandles _overlappingpawn))
  (Utilities|IsValid OverlappingPawn
    (:Is%20Valid
      (bind _output_get (LocalVariables|SetASC (Ability|GetFortAbilitySystemComponent OverlappingPawn)))
      (Utilities|IsValid _output_get
        (:Is%20Valid
          (if _returnvalue
            (for _array_element (Utilities|Struct|BreakBPSFZGameplayEffectHandles _value)
              (GameplayEffects|RemoveActiveGameplayEffect (LocalVariables|GetASC) _array_element 1))
            (Utilities|Map|Remove (Variables|Default|GetActorToGameplayEffectHandles) (Variables|GetOverlappingPawn))))))))

(fn Set_RippleFX_Param
  (params (HitWave-Loc :type Vector :default "0, 0, 0") (HitWave-SizeMultiplierOuterRing :type Float%20%28double-precision%29) (HitWave-SizeMultiplierInnerRing :type Float%20%28double-precision%29) (HitWave-DurationNew :type Float%20%28double-precision%29) (HitGridBrightness :type Float%20%28double-precision%29))
  (bind _safezone_in (Variables|Default|GetSafeZone_IN))
  (|CallParentFunction HitWave-Loc HitWave-SizeMultiplierOuterRing HitWave-SizeMultiplierInnerRing HitWave-DurationNew HitGridBrightness)
  (Rendering|Material|SetVectorParameterValueonMaterials _safezone_in "HitWave-Loc" HitWave-Loc)
  (Rendering|Material|SetScalarParameterValueonMaterials _safezone_in "GameTime" (Utilities|Time|GetGameTimeinSeconds))
  (Rendering|Material|SetScalarParameterValueonMaterials _safezone_in "HitWave-SizeMultiplierOuterRing" HitWave-SizeMultiplierOuterRing)
  (Rendering|Material|SetScalarParameterValueonMaterials _safezone_in "HitWave-SizeMultiplierInnerRing" HitWave-SizeMultiplierInnerRing)
  (Rendering|Material|SetScalarParameterValueonMaterials _safezone_in "HitWave-DurationNew" HitWave-DurationNew)
  (Rendering|Material|SetScalarParameterValueonMaterials _safezone_in "HitEffectLerp" HitGridBrightness))

(fn SetIgnoreActor
  (params (Actor :type Actor%20Object%20Reference))
  (bind _actorstoignorewhilemoving (Variables|Default|GetActorsToIgnoreWhileMoving))
  (Utilities|IsValid Actor
    (:Is%20Valid
      (if (not (Utilities|Array|ContainsItem _actorstoignorewhilemoving Actor))
        (Utilities|Array|Add _actorstoignorewhilemoving Actor)
        (Collision|IgnoreActorwhenMoving (Variables|Default|GetCollisionSphere) Actor true)
        (Game|BindEventtoOnDestroyed Actor (EventDispatchers|CreateEvent :!event OnMoveIgnoreActorDestroyed))
        (bind _asfort_athena_vehicle (Utilities|Casting|CastToFortAthenaVehicle Actor))
        (Vehicle|BindEventtoOnPawnExitedVehiclefromSeat _asfort_athena_vehicle (EventDispatchers|CreateEvent :!event OnPawnExitCollisionIgnoredVehicle))))))

(fn SetupOverlappingGameplayEffectsToApply
  (Networking|SwitchHasAuthority
    (:Authority
      (Utilities|FlowControl|Sequence
        (:then_0
          (if (ScalableFloat|GetValueasBool (Variables|Default|GetHolsterWeaponHF))
            (Utilities|Array|Add (Variables|Default|GetOverlappingGameplayEffectsToApply) (Variables|Default|GetHolsterWeaponGameplayEffect))))
        (:then_1
          (if (ScalableFloat|GetValueasBool (Variables|Default|GetEnableLowGravity))
            (Utilities|Array|Add (Variables|Default|GetOverlappingGameplayEffectsToApply) (Variables|Default|GetLowGravityGameplayEffect))))
        (:then_2
          (Utilities|Array|Add (Variables|Default|GetOverlappingGameplayEffectsToApply) (Variables|Default|GetPlayerOverlappingGameplayEffect))
          (Utilities|Array|Add (Variables|Default|GetOverlappingGameplayEffectsToApply) (Variables|Default|GetDamageImmunityGameplayEffect))
          (Utilities|Array|Add (Variables|Default|GetOverlappingGameplayEffectsToApply) (Variables|Default|GetBlockTargetingGameplayEffect))
          (Utilities|Array|Add (Variables|Default|GetOverlappingGameplayEffectsToApply) (Variables|Default|GetAntiCheatGameplayEffect)))))))

(fn TryBlockBuilding
  (if (ScalableFloat|GetValueasBool (Variables|Default|GetBlockBuilding))
    (bind _outactors (Collision|SphereOverlapActors (Transformation|GetWorldLocation (Variables|Default|GetCollisionSphere)) (Components|Sphere|GetScaledSphereRadius (Variables|Default|GetCollisionSphere)) (Utilities|Array|MakeArray "ObjectTypeQuery1") 0 (Utilities|Array|MakeArray self)))
    (for _array_element _outactors
      (Utilities|IsA(soft) _array_element "/Script/FortniteGame.BuildingSMActor"
        (:True
          (Actor|DestroyActor _array_element))))
    (else
      (Variables|Building|SetbForceBlockBuildings false)
      (Variables|Building|SetbDoNotBlockBuildings true))))

(fn ConstructionScript
  (|CallParentFunction)
  (Transformation|SetWorldScale3D (Variables|Default|GetRoot) (Math|Conversions|ToVector(Float) (Math|Float|SafeDivide (ScalableFloat|GetValueatLevel (Variables|SilverBronco|GetMaxScale)) (ScalableFloat|GetValueatLevel (Variables|Default|GetMaxScaleParent)))))
  (Rendering|Material|SetScalarParameterValueonMaterials (Variables|Default|GetSafeZone) "OverallOpacity" 0.6))
