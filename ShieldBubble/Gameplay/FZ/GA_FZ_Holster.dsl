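; Fortnite gameplay ability that holsters the pawn's weapon and blocks attacks while active, undoing both on end.
; Activation applies a no-weapon pose effect and re-evaluates on each weapon equip; a consumables toggle exempts weapons with do-not-holster tags.
; An optional reload path unholsters on local reload attempts, replicates the reload, then polls on next-tick timers to reholster once it finishes.
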(event Ability|EventActivateAbility
  ;[1 1 1] Activate - Holster
  (Variables|Default|SetHolster%20Id (Utilities|String|StringToName (Utilities|GetDisplayName self)))
  (FortPawn|BindEventtoOnWeaponEquippedDelegate (Ability|GetActivatingPawn) (delegate Custom|WeaponEquipped))
  (GameplayEffects|ApplyGameplayEffectToSelf :self (Ability|GetFortAbilitySystemComponent (Ability|GetActivatingPawn)) :GameplayEffectClass (Variables|Default|GetNoWeaponPoseEffect) :Level 0.0)
  (if (CallFunction|ShouldHolsterWeapon (Variables|/Script/FortniteGame.FortPawn|GetCurrentWeapon (Ability|GetActivatingPawn)))
    (if (ScalableFloat|GetValueasBool (Variables|Default|GetAllowReloadingHF))
      (if (FortWeapon|IsReloading (Variables|/Script/FortniteGame.FortPawn|GetCurrentWeapon (Ability|GetActivatingPawn)))
        (CallFunction|SetPreventAttack true)
        (Ability|IsLocallyControlled
          (:True
            (CallFunction|CheckFinishedReloading)
            (EventDispatchers|BindEventtoOnTryReload (Variables|/Script/FortniteGame.FortPawn|GetCurrentWeapon (Ability|GetActivatingPawn)) (EventDispatchers|CreateEvent :!event OnLocalClientTryReload))))
        (else
          _))
      (else
        _))
    (CallFunction|SetCurrentWeaponHolstered true)))

(event Custom|WeaponEquipped
  (params (NewWeapon :type Fort%20Weapon%20Object%20Reference) (PrevWeapon :type Fort%20Weapon%20Object%20Reference))
  ;[1 1 1] Activate - Holster
  (CharacterAnimation|SetWeaponPawnAnimsetOverride PrevWeapon 0)
  (EventDispatchers|UnbindEventfromOnTryReload PrevWeapon (EventDispatchers|CreateEvent :!event OnLocalClientTryReload))
  (Utilities|Time|ClearandInvalidateTimerbyHandle (Variables|Default|GetReloadTickTimer))
  (if (CallFunction|ShouldHolsterWeapon NewWeapon)
    (CallFunction|SetCurrentWeaponHolstered true)
    (else
      (CallFunction|SetCurrentWeaponHolstered false true))))

(event Ability|EventOnEndAbility
  (params (bWasCancelled :type Boolean))
  (bind _returnvalue (Ability|GetActivatingPawn))
  (bind _returnvalue_2 (Ability|GetFortAbilitySystemComponent _returnvalue))
  ;[1 1 1] On End Ability - Unholster
  (CallFunction|SetCurrentWeaponHolstered false true)
  (GameplayEffects|RemoveActiveGameplayEffectbySourceEffect _returnvalue_2 (Variables|Default|GetNoWeaponPoseEffect) 0 -1)
  (GameplayEffects|RemoveActiveGameplayEffectbySourceEffect _returnvalue_2 (Variables|Default|GetPreventAttackEffect) 0 -1))

(event Custom|Rep_ClientTryReload
  (params (Weapon :type Fort%20Weapon%20Object%20Reference))
  ; Reload
  (Weapon|SetHolsterWeapon (Ability|GetActivatingPawn) 0 false false (Variables|Default|GetHolster%20Id))
  (FortWeapon|Reload Weapon))

(event Custom|OnClientReholsterWeapon
  ; Reload
  (if (CallFunction|ShouldHolsterWeapon (Variables|/Script/FortniteGame.FortPawn|GetCurrentWeapon (Ability|GetActivatingPawn)))
    (CallFunction|SetCurrentWeaponHolstered true)))

(event Custom|WaitFinishReloading
  ; Reload
  (bind _returnvalue (Utilities|Time|SetTimerforNextTickbyEvent (delegate Custom|CheckFinishedReloading)))
  (Variables|Default|SetReloadTickTimer _returnvalue))

(event Custom|CheckFinishedReloading
  ; Reload
  (if (FortWeapon|IsReloading (Variables|/Script/FortniteGame.FortPawn|GetCurrentWeapon (Ability|GetActivatingPawn)))
    (Variables|Default|SetStartedReloading true)
    (CallFunction|WaitFinishReloading)
    (elif (Variables|Default|GetStartedReloading)
      (Variables|Default|SetStartedReloading false)
      (Utilities|Time|ClearandInvalidateTimerbyHandle (Variables|Default|GetReloadTickTimer))
      (CallFunction|OnClientReholsterWeapon)
      (if (CallFunction|ShouldHolsterWeapon (Variables|/Script/FortniteGame.FortPawn|GetCurrentWeapon (Ability|GetActivatingPawn)))
        (CallFunction|SetCurrentWeaponHolstered true))
      (else
        (CallFunction|WaitFinishReloading)))))

(fn OnLocalClientTryReload
  (params (Weapon :type Fort%20Weapon%20Object%20Reference))
  (if (ScalableFloat|GetValueasBool (Variables|Default|GetAllowReloadingHF))
    (if (Math|Boolean|ANDBoolean (not (FortWeapon|IsReloading Weapon)) (< (FortWeapon|GetLocalMagazineAmmoCount Weapon) (FortWeapon|GetBulletsPerClip Weapon)) (>= (FortWeapon|GetRemainingAmmo Weapon) (FortWeapon|GetBulletsPerClip Weapon)))
      (Weapon|SetHolsterWeapon (Ability|GetActivatingPawn) 0 false false (Variables|Default|GetHolster%20Id))
      (Variables|Default|SetStartedReloading false)
      (CallFunction|CheckFinishedReloading)
      (CallFunction|RepClientTryReload (Variables|GetWeapon)))))

(fn SetCurrentWeaponHolstered
  (params (Holstered :type Boolean) (PlayEquipAnim :type Boolean))
  (bind _returnvalue (Ability|GetActivatingPawn))
  (bind _returnvalue_2 (Ability|GetActivatingPawn))
  (bind _currentweapon (Variables|/Script/FortniteGame.FortPawn|GetCurrentWeapon _returnvalue))
  (bind _currentweapon_2 (Variables|/Script/FortniteGame.FortPawn|GetCurrentWeapon _returnvalue_2))
  (Utilities|IsValid (Ability|GetActivatingPawn)
    (:Is%20Valid
      (if Holstered
        (Class|FortWeapon|SetDisableEquipAnimation _currentweapon_2 (not (Variables|GetPlayEquipAnim)))
        (CharacterAnimation|SetWeaponPawnAnimsetOverride _currentweapon_2 "/Game/Animation/Game/MainPlayer/AnimAssets/FortPlayerDefault_AnimAsset.FortPlayerDefault_AnimAsset")
        (Weapon|SetHolsterWeapon _returnvalue_2 0 true (Variables|GetPlayEquipAnim) (Variables|Default|GetHolster%20Id))
        (CallFunction|SetPreventAttack true)
        (if (ScalableFloat|GetValueasBool (Variables|Default|GetAllowReloadingHF))
          (EventDispatchers|BindEventtoOnTryReload (Variables|/Script/FortniteGame.FortPawn|GetCurrentWeapon (Ability|GetActivatingPawn)) (EventDispatchers|CreateEvent :!event OnLocalClientTryReload)))
        (else
          (Class|FortWeapon|SetDisableEquipAnimation _currentweapon (not (Variables|GetPlayEquipAnim)))
          (CharacterAnimation|SetWeaponPawnAnimsetOverride _currentweapon 0)
          (Weapon|SetHolsterWeapon (Ability|GetActivatingPawn) 0 false (Variables|GetPlayEquipAnim) (Variables|Default|GetHolster%20Id))
          (CallFunction|SetPreventAttack false))))))

(fn SetPreventAttack
  (params (PreventAttack :type Boolean))
  (if PreventAttack
    (GameplayEffects|ApplyGameplayEffectToSelf :self (Ability|GetFortAbilitySystemComponent (Ability|GetActivatingPawn)) :GameplayEffectClass (Variables|Default|GetPreventAttackEffect) :Level 0.0)
    (else
      (GameplayEffects|RemoveActiveGameplayEffectbySourceEffect (Ability|GetFortAbilitySystemComponent (Ability|GetActivatingPawn)) (Variables|Default|GetPreventAttackEffect) 0 -1))))

(fn ShouldHolsterWeapon
  (params (Weapon :type Fort%20Weapon%20Object%20Reference))
  (Utilities|IsValid Weapon
    (:Is%20Valid
      (if (ScalableFloat|GetValueasBool (Variables|Default|GetAllowConsumablesHF))
        (if (GameplayTags|HasAnyTags (Tags|GetWeaponDataGameplayTags (Variables|GetWeapon)) (Variables|Default|GetDoNotHolsterTags))
          (return false)
          (else
            (return true)))
        (else
          (return true))))
    (:Is%20Not%20Valid
      (return false))))
