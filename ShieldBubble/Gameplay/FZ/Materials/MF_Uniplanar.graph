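# MF_Uniplanar is a library material function for cheap triplanar mapping, outputting one world-space UV set so materials need a single texture sample.
# Absolute normal components pick a projection plane, with the normal's sign adding per-side UV offsets.
# A pixel-shader temporal dither from screen position and noise, scaled by BlendSize, softens seams via TAA.
# It outputs triplanar and biplanar (side planes only) UVs and masks, plus an undithered biplanar mask.
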MaterialFunction /ShieldBubble/Gameplay/FZ/Materials/MF_Uniplanar
graph {
 properties {
  bExposeToLibrary=true
  description="Cheap triplanar. A single texture sample, blended using dither+TAA. Might require screenPosition derivatives to compute mip level"
 }
 nodes [
  n1 = FunctionInput(Preview=VertexNormalWS()) {bUsePreviewValueAsDefault=true, inputName=VertexNormalWS, sortPriority=2}
  n2 = Sign(None=@n1)
  n3 = ComponentMask(None=@n2) {g=true}
  n4 = Constant2Vector() {g=31.84, r=15.71}
  n5 = Multiply(
   A=Constant3Vector() {constant=<rgba 1 1 -1 0>}
   B=FunctionInput(Preview=WorldPosition().XYZ) {bUsePreviewValueAsDefault=true, inputName=WorldPosition, sortPriority=1}
  )
  n6 = ComponentMask(None=@n2) {r=true}
  n7 = Constant2Vector() {g=23.57, r=17.22}
  n8 = Add(A=Reroute(), B=Abs(None=@n1))
  n9 = ComponentMask(None=@n8) {g=true}
  n10 = ComponentMask(None=@n8) {r=true}
  n11 = Add(A=OneMinus(None=@n9), B=@n10)
  n12 = Constant() {r=0.3}
  n13 = ViewProperty() {property=MEVP_TemporalSampleIndex}
  MaterialExpressionAdd_14 = Add(A=ScreenPosition().PixelPosition, B=AppendVector(A=@n13.Property, B=@n13.Property))
  n14 = Custom!(p=@MaterialExpressionAdd_14) {
   code="Mod( ((uint)(p.x) + 2 * (uint)(p.y)) , 5 )"
   outputType=CMOT_Float1
   inputs [{inputName=p, input {expression=&self:MaterialExpressionAdd_14}}]
  }
  n15 = Constant()
  n16 = Add(
   A=MaterialFunctionCall!(TextureResolution=Constant2Vector() {g=64, r=64}) {materialFunction=/Engine/Functions/Engine_MaterialFunctions02/Texturing/ScreenAlignedPixelToPixelUVs}.Result
   B=Add(A=ComponentMask(None=@n15) {g=true, r=true}, B=ComponentMask(None=@n15) {b=true})
  )
  n17 = Divide(A=Add(A=@n14, B=TextureSample(UVs=@n16) {samplerType=SAMPLERTYPE_LinearGrayscale, texture=/Engine/EngineMaterials/Good64x64TilingNoiseHighFreq}.R) {constB=0}) {constB=6}
  n18 = Constant() {r=1}
  n19 = Add(A=Add(A=Multiply(A=@n17, B=@n18), B=Constant() {r=0.1})) {constB=-0.5}
  n20 = ShaderStageSwitch(
   PixelShader=Multiply(
    A=FunctionInput(Preview=@n12) {bUsePreviewValueAsDefault=true, inputName=BlendSize, inputType=FunctionInput_Scalar}
    B=Multiply(A=@n19) {constB=2}
   )
   VertexShader=Constant()
  )
  n21 = Saturate(None=Floor(None=Add(A=@n20, B=@n11)))
  n22 = LinearInterpolate(
   A=Add(A=Multiply(A=@n3, B=@n4), B=ComponentMask(None=@n5) {b=true, r=true})
   B=Add(A=Multiply(A=@n6, B=@n7), B=ComponentMask(None=@n5) {b=true, g=true})
   Alpha=@n21
  )
  n23 = Multiply(A=ComponentMask(None=@n2) {b=true}, B=Constant2Vector() {g=25.79, r=20.11})
  n24 = ComponentMask(None=@n5) {g=true, r=true}
  n25 = ComponentMask(None=@n8) {b=true}
  n26 = Add(A=OneMinus(None=Max(A=@n10, B=@n9)), B=@n25)
  n27 = Saturate(None=Floor(None=Add(A=@n20, B=@n26)))
  FunctionOutput(None=LinearInterpolate(A=@n22, B=Add(A=@n23, B=@n24), Alpha=@n27)) {outputName=Triplanar UVs}
  n28 = OneMinus(None=Max(A=@n9, B=@n25) {desc=Readded Z for mask})
  n29 = Floor(None=Add(B=Add(A=@n28, B=@n10)))
  n30 = OneMinus(None=Max(A=@n10, B=@n25))
  n31 = Floor(None=Add(A=@n20, B=Add(A=@n30, B=@n9)))
  FunctionOutput(None=MaterialFunctionCall(R=Saturate(None=@n29), G=Saturate(None=@n31), B=@n27) {materialFunction=&AppendMany}.RGB) {outputName=Triplanar Mask, sortPriority=2}
  FunctionOutput(None=@n22) {outputName=Biplanar UVs, sortPriority=1}
  FunctionOutput(None=@n21) {outputName=Biplanar Mask, sortPriority=3}
  FunctionOutput(None=@n11) {outputName=Biplanar Mask No Dither, sortPriority=4}
  Max(A=@n9, B=@n25) {desc=Excluded Z for biplanar}
 ]
}
