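# Material function that returns an approximate inverse of SmoothStep for a scalar input.
# The output adds half the difference between the input and its SmoothStep to the input, avoiding trigonometry.
# A debug branch, not wired to the output, plots SmoothStep in red, the approximation in green and the exact arcsine-based inverse in blue for comparison.
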MaterialFunction /ShieldBubble/Gameplay/FZ/Materials/MF_SmoothStepInverse
graph {nodes [
  n1 = FunctionInput(Preview=ComponentMask(None=TextureCoordinate()) {r=true}) {inputType=FunctionInput_Scalar}
  n2 = MaterialFunctionCall(Alpha=@n1) {materialFunction=&SmoothStep}
  n3 = Add(A=@n1, B=Multiply(A=Subtract(A=@n1, B=@n2.Result)) {constB=0.5})
  FunctionOutput(None=@n3) {outputName="InverseSmoothStep (Aprox)", sortPriority=1}
  n4 = Constant3Vector() {constant=<rgba 1 0 0 0>}
  n5 = Constant4Vector() {constant=<rgba 0 1 0 1>}
  n6 = MaterialFunctionCall(Alpha=@n1) {materialFunction=&SmoothStep}
  n7 = Constant3Vector() {constant=<rgba 0 1 0 0>}
  n8 = Constant() {r=0.5}
  n9 = Subtract(B=Multiply(A=@n1) {constB=2})
  n10 = Sine(None=Divide(A=Arcsine!(None=@n9)) {constB=3})
  Max(
   A=Max(
    A=MaterialFunctionCall!(X range Y range=@n5.RGBA, Gradient Height, ShowGrid?, Thickness, Color=@n4, "f(x)"=@n6.Result) {materialFunction=/Engine/Functions/Engine_MaterialFunctions01/Debug/PlotFunctionOnGraph}."f(x) = y"
    B=MaterialFunctionCall(X range Y range=@n5.RGBA, Color=@n7, "f(x)"=@n3) {materialFunction=&PlotFunctionOnGraph}."f(x) = y"
   )
   B=MaterialFunctionCall(X range Y range=@n5.RGBA, Color=Constant3Vector() {constant=<rgba 0 0 1 0>}, "f(x)"=Subtract(A=@n8, B=@n10)) {materialFunction=&PlotFunctionOnGraph}."f(x) = y"
  )
 ]}
