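# M_DeluluSpiral is a material that builds an animated, stylized spiral with a purple, pink and yellow-green palette.
# Polar UVs, rotated over time and warped by radius, form stripes that offset a radial four-color gradient.
# A rotating, grid-quantized halftone texture drives procedural dot size and Bayer dithering, which darken the gradient between stripes alongside a tiled dot texture.
# No material output connections appear, and several experimental branches end unused.
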Material /ShieldBubble/Gameplay/FZ/Materials/M_DeluluSpiral
graph {
 properties {bAutomaticallySetUsageInEditor=false, usedShadingModels=Default Lit}
 nodes [
  n1 = Constant() {r=2}
  n2 = Constant() {r=64}
  n3 = Multiply(A=@n1, B=@n2)
  n4 = TextureCoordinate()
  n5 = Multiply(A=@n4, B=@n3)
  n6 = Rotator(Coordinate=Divide(A=Floor(None=@n5), B=@n3)) {speed=0.2}
  Add(A=@n6, B=Constant2Vector() {g=-0.25, r=0.25}) {constB=0.125}
  Add(B=Constant2Vector() {g=0.5, r=0.5})
  n7 = Multiply(A=@n1, B=@n4)
  n8 = Multiply(A=@n7, B=@n2)
  Add(A=Multiply(A=@n8) {constB=0.5}) {constB=0.25}
  n9 = ConstantBiasScale(None=TextureCoordinate()) {bias=-0.5, scale=2}
  Add(A=@n9, B=Constant2Vector() {g=0.5, r=0.5})
  ComponentMask(None=TextureCoordinate()) {r=true}
  Constant()
  n10 = Subtract(A=@n4, B=Constant() {r=0.5})
  n11 = Arctangent2!(Y=ComponentMask(None=@n10) {r=true}, X=ComponentMask(None=@n10) {g=true})
  n12 = Constant() {r=2}
  n13 = Frac(None=Divide(
    A=ComponentMask(None=@n11) {r=true}
    B=MaterialFunctionCall(Multiplier=@n12) {materialFunction=&Pi}.Result
   ))
  n14 = Time()
  n15 = Multiply(A=Constant() {r=-0.03}, B=@n14)
  n16 = Multiply(A=@n10, B=@n10)
  n17 = SquareRoot(None=Add(A=ComponentMask(None=@n16) {r=true}, B=ComponentMask(None=@n16) {g=true}))
  n18 = Constant()
  n19 = Multiply(A=Add(A=@n17, B=Multiply(A=@n14, B=@n18))) {constB=2}
  n20 = Add(A=Multiply(A=Add(A=@n13, B=@n15)), B=Multiply(A=Power(Base=@n19) {constExponent=0.05}) {constB=10})
  n21 = Frac(None=Multiply(A=Frac(None=@n20)) {constB=4})
  n22 = Multiply(A=@n21) {constB=2}
  Frac(None=@n22) {desc=uv}
  n23 = Constant() {r=0.5}
  n24 = Subtract(A=@n23) {constB=0.01}
  n25 = Add(A=@n23) {constB=0.01}
  n26 = ConstantBiasScale(None=Frac(None=@n8)) {bias=-0.5, scale=2}
  n27 = Constant()
  n28 = Constant() {r=1}
  n29 = MaterialFunctionCall(
   In=Power(Base=TextureSample(UVs=@n6) {samplerType=SAMPLERTYPE_LinearGrayscale, texture=&/T_PP_Halftone_Pattern}.R) {constExponent=0.8}
   Contrast=Constant() {r=0.1}
  ) {materialFunction=&CheapContrast}
  n30 = Subtract(
   A=Distance(A=@n26, B=Constant())
   B=LinearInterpolate(A=Subtract(A=@n27, B=@n28) {constB=0.01}, B=Add(A=@n27, B=@n28) {constB=0.01}, Alpha=@n29.Result) {constA=0.2, constB=0.8}
  )
  n31 = OneMinus(None=Saturate(None=SmoothStep(Min=@n24, Max=@n25, Value=@n30)))
  Frac(None=@n31)
  n32 = Divide(
   A=ComponentMask(None=WorldPosition().XYZ) {g=true, r=true}
   B=MaterialFunctionCall() {materialFunction=&ObjectScale}.Scale Z
  )
  Frac(None=Add(A=Divide(A=@n32) {constB=350}) {constB=0.5})
  Multiply(A=AppendVector(A=@n13, B=@n17), B=Multiply(A=Constant2Vector() {g=1, r=1}) {constB=10}) {constB=22}
  n33 = Saturate(None=OneMinus(None=@n19))
  n34 = LinearInterpolate(Alpha=@n33) {constA=-0.5}
  n35 = Constant3Vector() {constant=<rgba 0.8039 0.9804 0.0039 1>}
  n36 = Ceil(None=Add(A=@n22) {constB=-1}) {desc=mask}
  n37 = Multiply(A=ScalarParameter() {defaultValue=0.2, parameterName=Stripes gradient offset}, B=@n36)
  n38 = MaterialFunctionCall!(
   A=Constant3Vector() {constant=<rgba 0.2471 0.0275 0.451 1>}
   B=Constant3Vector() {constant=<rgba 0.702 0.1373 0.8863 1>}
   C=Constant3Vector() {constant=<rgba 1 0.1255 0.2902 1>}
   D=@n35
   Alpha=MaterialFunctionCall!(In=Add(A=@n33, B=@n37)) {materialFunction=&/MF_SmoothStepInverse}."InverseSmoothStep (Aprox)"
  ) {materialFunction=&/Lerp_4Color}
  n39 = LinearInterpolate(A=@n38.Result, B=@n35, Alpha=Saturate(None=Power(Base=@n34) {constExponent=3})) {constB=0}
  n40 = LinearInterpolate(Alpha=@n33) {constA=3, constB=1.5}
  n41 = Multiply(A=@n4) {constB=10}
  n42 = Multiply(A=OneMinus(None=TextureSample(UVs=@n41) {texture=&/T_Dots}.R)) {constB=0.25}
  n43 = TextureSample(UVs=Multiply(A=@n7)) {samplerType=SAMPLERTYPE_LinearGrayscale, texture=/InterchangeAssets/Utilities/T_Bayer64_Grayscale_64px}
  n44 = Step(Y=@n43.R, X=LinearInterpolate(Alpha=@n29.Result) {constA=-0.2, constB=1.2})
  n45 = Add(A=@n31) {constB=-0.5}
  n46 = Multiply(A=Saturate(None=Multiply(A=@n45, B=@n44)), B=OneMinus(None=@n36))
  n47 = ScalarParameter() {defaultValue=1, parameterName=Dither gradient offset}
  Multiply(A=OneMinus(None=Multiply(A=@n46, B=@n47)), B=Multiply(A=OneMinus(None=@n42), B=Power(Base=@n39, Exp=@n40)))
  Multiply(A=@n38.Result)
  OneMinus(None=@n29.Result)
  n48 = TextureCoordinate()
  n49 = Frac(None=Add(A=ConstantBiasScale(None=@n48) {bias=-0.5, scale=2}, B=Constant2Vector() {g=0.5, r=0.5}))
  n50 = Constant() {r=0.5}
  OneMinus(None=Multiply(A=Distance(A=@n49, B=@n50)) {constB=2})
  n51 = Frac(None=Multiply(A=@n9) {constB=0.5})
  n52 = Constant() {r=0.5}
  OneMinus(None=Multiply(A=Distance(A=@n51, B=@n52)) {constB=2})
  OneMinus(None=@n43.R)
  n53 = Multiply(A=@n26) {constB=0.5}
  n54 = Multiply(A=Distance(A=Frac(None=@n53), B=Constant() {r=0.5})) {constB=2}
  OneMinus(None=@n54)
  Reroute(None=@n54)
  Reroute(None=@n45)
  n55 = Sine(None=@n21)
  n56 = Constant() {r=1}
  Saturate(None=OneMinus(None=MaterialFunctionCall(In=@n55, Contrast=@n56) {materialFunction=&CheapContrast}.Result)) {desc=soft edges}
 ]
}
