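# Translucent unlit material for the SilverBlazer variant of a sphere shield bubble, with deploy, shrink, dissolve and hit-reaction effects.
# A "Delulu" layer builds a camera-tilted polar spiral in four colors, overlaid with distance-faded halftone dither and tinted by time of day.
# Fresnel color, intersection glow and grid layers accompany timestamp-driven hit rings that distort UVs and push vertices.
# Custom HLSL provides seam-aware cube-face UVs; quality and feature-level switches strip costly effects.
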Material /ShieldBubble/Gameplay/FZ/Materials/M_SphereShield_SilverBlazer_FZ
graph {
 properties {blendMode=BLEND_Translucent, pixelDepthOffsetMode=PDOM_Legacy, shadingModel=MSM_Unlit, usedShadingModels=Unlit}
 nodes [
  n1 = NamedRerouteUsage()
  n2 = ComponentMask(None=@n1) {g=true}
  Abs(None=@n2)
  n3 = ScalarParameter() {defaultValue=300, group=HitWave, parameterName=HitWave-SizeMultiplierOuterRing}
  n4 = Time()
  n5 = ScalarParameter() {defaultValue=1, desc=Time-stamp, group=HitWave, parameterName=GameTime}
  n6 = Saturate(None=Divide(A=Subtract(A=@n4, B=@n5), B=ScalarParameter() {defaultValue=0.5, group=HitWave, parameterName=HitWave-DurationNew}))
  n7 = Power(Base=@n6, Exp=ScalarParameter() {defaultValue=0.5, group=HitWave, parameterName=HitWave-OuterRingDelay}) {constExponent=0.7}
  n8 = ComponentMask(None=VectorParameter() {group=HitWave, parameterName=HitWave-Loc, defaultValue=<rgba 100 -1400 -150 0>}.RGB) {b=true, g=true, r=true}
  n9 = WorldPosition() {worldPositionShaderOffset=WPT_ExcludeAllShaderOffsets}
  n10 = ScalarParameter() {defaultValue=-0.1, group=HitWave, parameterName=HitWave-InnerRingDelay}
  n11 = Multiply(
   A=Saturate(None=LinearInterpolate(A=@n10, Alpha=@n6) {constA=-0.7})
   B=ScalarParameter() {defaultValue=1000, group=HitWave, parameterName=HitWave-SizeMultiplierInnerRing}
  )
  n12 = ScalarParameter() {defaultValue=0.25, group=HitWave, parameterName=Hitwave-SphereHard2}
  n13 = Multiply(
   A=SphereMask(
    A=@n9.XYZ
    B=@n8
    Radius=Multiply(A=@n3, B=@n7)
    Hardness=ScalarParameter() {defaultValue=0.7, group=HitWave, parameterName=Hitwave-SphereHard1}
   ) {hardnessPercent=0}
   B=OneMinus(None=SphereMask(A=@n9.XYZ, B=@n8, Radius=@n11, Hardness=@n12) {hardnessPercent=0})
  )
  n14 = Constant()
  n15 = FeatureLevelSwitch(Default=@n13, ES2=@n14, ES3_1=@n14, SM4_REMOVED=@n13)
  n16 = OneMinus(None=@n6)
  n17 = Multiply(A=@n15, B=LinearInterpolate(B=ScalarParameter() {defaultValue=5, group=HitWave, parameterName=Hitwave-DistortAmt3}, Alpha=@n16))
  n18 = ScalarParameter() {defaultValue=1, parameterName=HitEffectLerp}
  Add(A=NamedRerouteDeclaration(None=Multiply(A=@n17, B=@n18)) {name="(Delulu) HitDistort", nodeColor=<rgba 1 0.5156 0 1>})
  Add(B=Constant2Vector() {g=0.5, r=0.5})
  Add(A=Multiply(A=Reroute()) {constB=0.5}) {constB=0.25}
  n19 = ConstantBiasScale(None=TextureCoordinate()) {bias=-0.5, scale=2}
  Add(A=@n19, B=Constant2Vector() {g=0.5, r=0.5})
  n20 = ComponentMask(None=PixelNormalWS()) {g=true, r=true}
  Add(A=Add(A=Multiply(A=@n20) {constB=-0.5}) {constB=0.5}, B=Constant2Vector())
  n21 = MaterialFunctionCall() {materialFunction=&ComponentPivotLocation}
  n22 = Subtract(A=@n21.Result, B=CameraPositionWS())
  AppendVector(A=ComponentMask(None=@n22) {g=true, r=true}, B=Constant())
  n23 = Multiply(A=MaterialFunctionCall() {materialFunction=&CameraDirectionVector}.Result) {constB=1500}
  n24 = Add(A=Add(A=@n23, B=@n21.Result), B=Constant3Vector() {constant=<rgba 0 0 100 0>}) {constB=0}
  n25 = MaterialFunctionCall(World Position=@n24) {materialFunction=&TransformToClipSpace}
  Clamp(None=@n25.Clip Space XY) {minDefault=-1}
  Constant4Vector() {constant=<rgba 0.8963 0.402 1 0>}
  Constant() {r=1}
  Constant() {r=10000}
  MaterialExpressionWorldPosition_1 = WorldPosition()
  MaterialExpressionObjectPositionWS_1 = ObjectPositionWS()
  MaterialExpressionScalarParameter_63 = ScalarParameter() {defaultValue=0.5, parameterName=til test}
  MaterialExpressionConstant_48 = Constant()
  n26 = Custom!(
   WorldPos=@MaterialExpressionWorldPosition_1.XYZ
   ObjectPos=@MaterialExpressionObjectPositionWS_1
   Tiling=@MaterialExpressionScalarParameter_63
   FlipU=@MaterialExpressionConstant_48
  ) {
   code="float3 P = WorldPos - ObjectPos;\r\nfloat3 D = normalize(P);\r\nfloat3 A = abs(D);\r\nfloat m  = max(A.x, max(A.y, A.z));\r\nfloat3 C = D / m; // on the dominant face plane (±1)\r\n\r\nfloat u,v;\r\n// Face selection (ties break toward later checks; matches UE If-node order)\r\nif (A.x >= A.y &&&& A.x >= A.z) {\r\n    // X major\r\n    if (D.x > 0) { u = -C.z; v =  C.y; }   // +X\r\n    else         { u =  C.z; v =  C.y; }   // -X\r\n}\r\nelse if (A.y >= A.z) {\r\n    // Y major\r\n    if (D.y > 0) { u =  C.x; v = -C.z; }   // +Y\r\n    else         { u =  C.x; v =  C.z; }   // -Y\r\n}\r\nelse {\r\n    // Z major\r\n    if (D.z > 0) { u =  C.x; v =  C.y; }   // +Z\r\n    else         { u = -C.x; v =  C.y; }   // -Z\r\n}\r\n\r\nfloat2 UV = float2(u, v) * 0.5 * Tiling;\r\n\r\n// Optional inside flip to correct mirror, keeps seams intact\r\nif (FlipU > 0.5) UV.x = -UV.x;\r\n\r\nfloat3 UVf = float3(u, v, 0.0) * 0.5 * Tiling;\r\nreturn UVf;\r\n"
   showCode=true
   inputs [
    {
     inputName=WorldPos
     input {
      expression=&self:MaterialExpressionWorldPosition_1
      mask=1
      maskB=1
      maskG=1
      maskR=1
     }
    }
    {
     inputName=ObjectPos
     input {expression=&self:MaterialExpressionObjectPositionWS_1}
    }
    {inputName=Tiling, input {expression=&self:MaterialExpressionScalarParameter_63}}
    {inputName=FlipU, input {expression=&self:MaterialExpressionConstant_48}}
   ]
  }
  n27 = ScreenPosition()
  n28 = MaterialFunctionCall(Float2=ViewSize()) {materialFunction=&BreakOutFloat2Components}
  n29 = AppendVector(A=Constant() {r=1}, B=Divide(A=@n28.G, B=@n28.G))
  n30 = ConstantBiasScale(None=Multiply(A=ConstantBiasScale(None=@n27.ViewportUV) {bias=-0.5, scale=2}, B=@n29))
  n31 = Multiply(A=@n29, B=Multiply(A=Constant2Vector() {g=-0.5, r=0.5}, B=@n25.Clip Space XY) {constA=1})
  n32 = MaterialFunctionCall(
   Fade Length=ScalarParameter() {defaultValue=20000, group=Delulu, parameterName=FadeLength}
   Fade Offset=ScalarParameter() {defaultValue=3000, group=Delulu, parameterName=FadeOffset}
  ) {materialFunction=&CameraDepthFade}
  n33 = MaterialFunctionCall(Alpha=@n32.Result) {materialFunction=&SmoothStep}
  n34 = Add(A=MaterialFunctionCall(UVs=Subtract(A=@n30, B=@n31) {constB=0}, Texture Scale=LinearInterpolate(Alpha=@n33.Result) {constA=1, constB=0.25}) {materialFunction=&ScaleUVsByCenter}.UVs) {constB=-0.5}
  n35 = LinearInterpolate(
   A=ComponentMask(None=@n26) {g=true, r=true}
   B=Multiply(A=Constant() {r=0.5}, B=Multiply(A=@n34) {constB=2}) {constA=1}
   Alpha=Saturate(None=NamedRerouteUsage())
  )
  n36 = Rotator(Coordinate=@n35) {centerX=0, centerY=0, speed=0.2}
  n37 = Constant() {r=3}
  n38 = Constant() {r=64}
  n39 = Multiply(A=@n37, B=@n38)
  Divide(A=Floor(None=Multiply(A=@n36, B=@n39)), B=@n39)
  n40 = PixelDepth()
  n41 = Constant() {r=-100}
  n42 = Saturate(None=Divide(A=Add(A=@n40.R, B=@n41), B=ScalarParameter() {defaultValue=300, parameterName=NearFade}))
  n43 = ScalarParameter() {defaultValue=100, group=IntersectingEdge, parameterName=IntersectingEdgeBias}
  n44 = ScalarParameter() {defaultValue=800, group=IntersectingEdge, parameterName=IntersectingEdgeDivisor}
  n45 = Multiply(A=LinearInterpolate(A=NamedRerouteUsage()) {constAlpha=0.1}, B=NamedRerouteUsage())
  n46 = NamedRerouteUsage()
  n47 = VectorParameter() {group=IntersectingEdge, parameterName=IntersectingEdgeColor, defaultValue=<rgba 0.0096 0.1133 0.1927 1>}
  n48 = Multiply(A=LinearInterpolate(Alpha=Saturate(None=@n46)) {constA=0.5}, B=@n47.RGB)
  n49 = Multiply(
   A=Power(
    Base=MaterialFunctionCall!(Distance Bias=@n43, Distance Divisor=@n44, Scene Depth, Pixel Depth) {materialFunction=/Game/Effects/Fort_Effects/Materials/_Functions/MF_DistanceBehindTranslucency_01}.Normalized Inverted
    Exp=ScalarParameter() {defaultValue=4, group=IntersectingEdge, parameterName=Intersection Edge Exponent}
   )
   B=Multiply(A=Multiply(A=@n45, B=@n48), B=ScalarParameter() {defaultValue=1.6779, group=IntersectingEdge, parameterName=IntBrightness})
  )
  n50 = Multiply(A=@n42, B=Add(A=@n49))
  n51 = ScalarParameter() {defaultValue=0.4238, group=Opacity, parameterName=OpacityInside}
  n52 = ScalarParameter() {defaultValue=0.5762, group=Opacity, parameterName=OpacityOutside}
  n53 = Saturate(None=NamedRerouteUsage())
  n54 = Multiply(A=@n42, B=Add(A=@n49, B=LinearInterpolate(A=@n51, B=@n52, Alpha=@n53)))
  FeatureLevelSwitch(Default=@n50, ES2=@n50, ES3_1=@n50, SM4_REMOVED=@n50, SM5=@n54, SM6=@n54)
  n55 = Constant() {r=0.5}
  n56 = Subtract(A=@n55) {constB=0.01}
  n57 = Add(A=@n55) {constB=0.01}
  n58 = Multiply(A=@n37, B=@n35)
  n59 = Frac(None=Multiply(A=@n38, B=@n58))
  n60 = Constant()
  n61 = Constant() {r=1}
  n62 = Saturate(None=NamedRerouteUsage())
  n63 = Add(A=Multiply(A=TextureCoordinate(), B=LinearInterpolate(Alpha=@n62) {constA=3, constB=5}) {constB=3}, B=Multiply(A=Time()) {constB=-0.05})
  n64 = MaterialFunctionCall(
   In=Power(Base=TextureSample(UVs=@n63) {samplerType=SAMPLERTYPE_LinearGrayscale, texture=&/T_PP_Halftone_Pattern}.R) {constExponent=0.8}
   Contrast=Constant() {r=0.1}
  ) {materialFunction=&CheapContrast}
  n65 = Subtract(
   A=Distance(A=ConstantBiasScale(None=@n59) {bias=-0.5, scale=2}, B=Constant())
   B=LinearInterpolate(A=Subtract(A=@n60, B=@n61) {constB=0.01}, B=Add(A=@n60, B=@n61) {constB=0.01}, Alpha=@n64.Result) {constA=0.2, constB=0.8}
  )
  n66 = OneMinus(None=Saturate(None=SmoothStep(Min=@n56, Max=@n57, Value=@n65)))
  Frac(None=@n66)
  n67 = Constant()
  Frac(None=Multiply(A=Distance(A=@n35, B=@n67)) {constB=50})
  n68 = OneMinus(None=Sine(None=Multiply(A=@n2) {constB=0.5}))
  n69 = ComponentMask(None=@n1) {r=true}
  n70 = MaterialFunctionCall(Speed=Constant() {r=0.5}) {materialFunction=&DebugTimeSine}
  n71 = OneMinus(None=Power(Base=OneMinus(None=@n70.Result)) {constExponent=1})
  n72 = ScalarParameter() {parameterName=Deploy}
  n73 = Multiply(A=LinearInterpolate(Alpha=@n69) {constA=0.7}, B=OneMinus(None=StaticSwitchParameter(True=@n71, False=@n72) {parameterName=Debug}))
  n74 = LinearInterpolate(Alpha=@n73) {constA=-1, constB=2}
  LinearInterpolate(Alpha=Saturate(None=Subtract(A=@n68, B=@n74))) {constB=70}
  n75 = ScalarParameter() {defaultValue=2, group=Delulu, parameterName=Center Opacity Scale}
  n76 = PixelNormalWS()
  n77 = CameraVectorWS()
  n78 = OneMinus(None=Saturate(None=DotProduct(A=@n76, B=@n77)))
  n79 = Power(Base=Saturate(None=Multiply(A=@n75, B=@n78)), Exp=ScalarParameter() {defaultValue=1, group=Delulu, parameterName=Center Opacity Pow})
  n80 = LinearInterpolate(
   A=ScalarParameter() {defaultValue=0.3, group=Delulu, parameterName=Center Opacity}
   B=ScalarParameter() {defaultValue=1, group=Delulu, parameterName=Edges Opacity}
   Alpha=MaterialFunctionCall(Alpha=Saturate(None=@n79)) {materialFunction=&SmoothStep}.Result
  )
  LinearInterpolate(B=@n80, Alpha=Saturate(None=NamedRerouteUsage())) {constA=1}
  n81 = Subtract(A=@n24, B=CameraPositionWS())
  n82 = Length(None=@n81) {is=class MaterialExpressionLength}
  n83 = Multiply(A=Add(A=@n82) {constB=-1250}) {constB=0.0005}
  LinearInterpolate(Alpha=MaterialFunctionCall(Alpha=Saturate(None=@n83)) {materialFunction=&SmoothStep}.Result) {constA=1, constB=0.5}
  LinearInterpolate(A=@n20, B=Normalize(VectorInput=@n20), Alpha=Multiply() {constA=1, constB=-0.5}) {constAlpha=-0.5}
  n84 = Constant()
  n85 = Saturate(None=Multiply(A=Distance(A=@n20, B=@n84)) {constB=2})
  LinearInterpolate(Alpha=OneMinus(None=Power(Base=@n85)))
  n86 = WorldPosition() {worldPositionShaderOffset=WPT_ExcludeAllShaderOffsets}
  n87 = ObjectPositionWS()
  n88 = MaterialFunctionCall(Vector or UVs=ComponentMask(None=Subtract(A=@n86.XYZ, B=@n87)) {g=true, r=true}) {materialFunction=&VectorToRadialValue}
  n89 = Power(Base=ComponentMask(None=@n86.XYZ) {b=true}, Exp=ScalarParameter() {defaultValue=2, group=Delulu, parameterName=cyl tiling side pow})
  n90 = AppendVector(
   A=ScalarParameter() {defaultValue=1, group=Delulu, parameterName=cyl tiling side x}
   B=ScalarParameter() {defaultValue=20000000, group=Delulu, parameterName=cyl tiling side y}
  )
  n91 = LinearInterpolate(
   A=ScalarParameter() {defaultValue=1, group=Delulu, parameterName=cyl tiling xy near}
   B=ScalarParameter() {defaultValue=5, group=Delulu, parameterName=cyl tiling xy far}
  )
  n92 = ScalarParameter() {defaultValue=7000, group=Delulu, parameterName=cyl tiling top xy}
  n93 = Transform(None=Constant3Vector() {constant=<rgba 0 0 1 0>})
  n94 = Add(
   A=Abs(None=ComponentMask(None=@n93) {b=true})
   B=Multiply(B=ScalarParameter() {defaultValue=1.4, group=Delulu, parameterName=cyl top offset}) {constA=-0.25}
  ) {constB=-0.25}
  LinearInterpolate(
   A=Divide(A=AppendVector(A=@n88.Vector Converted to Angle, B=@n89), B=Divide(A=@n90, B=@n91))
   B=Divide(A=ComponentMask(None=@n86.XYZ) {g=true, r=true}, B=Divide(A=@n92, B=@n91))
   Alpha=MaterialFunctionCall(In=Saturate(None=@n94), Contrast=Constant() {r=1111111}) {materialFunction=&CheapContrast}.Result
  )
  n95 = TextureCoordinate()
  LinearInterpolate(
   A=TextureSample(UVs=Multiply(A=@n95) {constB=2}) {samplerType=SAMPLERTYPE_Grayscale, texture=&LowResBlurredNoise}.R
   B=@n69
  )
  LinearInterpolate(A=@n23, B=Multiply(A=@n22) {constB=1500})
  MaterialFunctionCall(Float2=TextureCoordinate()) {materialFunction=&BreakOutFloat2Components}
  MaterialFunctionCall() {materialFunction=&CameraDirectionVector}
  n96 = Multiply(A=Add(A=@n82) {constB=-1250}) {constB=0.001}
  Multiply(A=@n31, B=MaterialFunctionCall(Alpha=Saturate(None=@n96)) {materialFunction=&SmoothStep}.Result)
  n97 = Constant3Vector() {constant=<rgba 0.8039 0.9804 0.0039 1>}
  n98 = Time()
  n99 = Multiply(A=Multiply(A=Constant() {r=-0.03}, B=ScalarParameter() {defaultValue=0.5, group=Delulu, parameterName=SpiralSpeed}), B=@n98)
  n100 = Constant() {r=0.5}
  n101 = MaterialFunctionCall() {materialFunction=&ComponentPivotLocation}
  n102 = Normalize(VectorInput=Subtract(A=CameraPositionWS(), B=@n101.Result))
  n103 = Constant3Vector() {constant=<rgba 0 0 1 0>}
  n104 = Normalize(VectorInput=MaterialFunctionCall!(Vector1=@n102, Vector2=@n103) {materialFunction=/Engine/Functions/Engine_MaterialFunctions02/Math/CreateThirdOrthogonalVector}.Vector3)
  n105 = Abs(None=DotProduct(A=@n103, B=@n102))
  n106 = Multiply(
   A=Multiply(A=ScalarParameter() {defaultValue=0.12, group=Delulu, parameterName=RotAng}, B=OneMinus(None=@n105))
   B=Saturate(None=NamedRerouteUsage())
  )
  n107 = WorldPosition()
  n108 = Subtract(
   A=MaterialFunctionCall() {materialFunction=&ComponentPivotLocation}.Result
   B=Add(A=RotateAboutAxis(NormalizedRotationAxis=@n104, RotationAngle=@n106, PivotPoint=@n101.Result, Position=@n107.XYZ), B=@n107.XYZ)
  )
  n109 = Divide(A=Divide(
    A=ComponentMask(None=@n108) {g=true, r=true}
    B=MaterialFunctionCall() {materialFunction=&ObjectScale}.Scale Z
   )) {constB=100}
  n110 = ScalarParameter() {defaultValue=1.1, group=Delulu, parameterName=SpiralScale}
  n111 = DotProduct(A=Multiply(A=NamedRerouteUsage(), B=PixelNormalWS()), B=Constant3Vector() {constant=<rgba 0 0 1 0>})
  n112 = LinearInterpolate(
   A=@n110
   B=Multiply(A=@n110, B=ScalarParameter() {defaultValue=0.8, group=Delulu, parameterName=SpiralScale edge})
   Alpha=Power(Base=OneMinus(None=@n111), Exp=ScalarParameter() {defaultValue=1.7, group=Delulu, parameterName=SpiralScale edge pow}) {constExponent=55}
  )
  n113 = Subtract(A=Add(A=Divide(A=@n109, B=@n112) {constB=1}) {constB=0.5}, B=@n100)
  n114 = Arctangent2!(Y=ComponentMask(None=@n113) {r=true}, X=ComponentMask(None=@n113) {g=true})
  n115 = Constant() {r=2}
  n116 = Frac(None=Divide(
    A=ComponentMask(None=@n114) {r=true}
    B=MaterialFunctionCall(Multiplier=@n115) {materialFunction=&Pi}.Result
   ))
  n117 = Constant()
  n118 = Multiply(A=@n113, B=@n113)
  n119 = SquareRoot(None=Add(A=ComponentMask(None=@n118) {r=true}, B=ComponentMask(None=@n118) {g=true}))
  n120 = Multiply(A=Add(A=Multiply(A=@n98, B=@n117), B=@n119)) {constB=2}
  n121 = Add(A=Multiply(A=Add(A=@n116, B=@n99)), B=Multiply(A=Power(Base=@n120) {constExponent=0.05}) {constB=10})
  n122 = Frac(None=Multiply(A=Frac(None=@n121), B=ScalarParameter() {defaultValue=5, group=Delulu, parameterName=Arms}) {constB=4})
  n123 = Sine(None=@n122)
  n124 = OneMinus(None=MaterialFunctionCall(In=Power(Base=@n123) {constExponent=0.3}, Contrast=Constant() {r=10}) {materialFunction=&CheapContrast}.Result)
  n125 = Multiply(A=@n111) {constB=-1}
  n126 = OneMinus(None=MaterialFunctionCall(Alpha=Saturate(None=@n125)) {materialFunction=&SmoothStep}.Result)
  n127 = DotProduct(A=Multiply(A=PixelNormalWS(), B=NamedRerouteUsage()), B=Constant3Vector() {constant=<rgba 0 0 1 0>})
  n128 = Saturate(None=OneMinus(None=@n120))
  n129 = OneMinus(None=@n128)
  n130 = Multiply(A=OneMinus(None=Power(Base=@n129) {constExponent=1}), B=Saturate(None=Sign(None=@n127)))
  n131 = Power(Base=@n130) {constExponent=5}
  n132 = LinearInterpolate(A=Saturate(None=Multiply(A=@n124, B=@n126)) {desc=soft edges}, Alpha=@n131)
  n133 = ScalarParameter() {defaultValue=0.2, group=Delulu, parameterName=Stripes gradient offset}
  n134 = MaterialFunctionCall() {materialFunction=&ComponentPivotLocation}
  n135 = Constant3Vector() {constant=<rgba 0 0 500 0>}
  n136 = Length(None=Subtract(A=Add(A=@n134.Result, B=@n135), B=CameraPositionWS()))
  n137 = Multiply(A=Add(A=@n136) {constB=-1800}) {constB=0.0015}
  n138 = OneMinus(None=MaterialFunctionCall(Alpha=Saturate(None=@n137)) {materialFunction=&SmoothStep}.Result)
  n139 = Multiply(A=Multiply(A=@n133) {constB=0.2}, B=@n138)
  n140 = Saturate(None=NamedRerouteUsage())
  n141 = Add(
   A=Multiply(A=@n132, B=LinearInterpolate(A=@n139, B=@n133, Alpha=@n140))
   B=LinearInterpolate(A=Constant() {r=0.2}, B=Constant() {r=0.05}, Alpha=@n140) {constA=0.2, constB=0}
  )
  n142 = MaterialFunctionCall!(
   A=Constant3Vector() {constant=<rgba 0.2462 0.0273 0.4508 1>}
   B=Constant3Vector() {constant=<rgba 0.702 0.1373 0.8863 1>}
   C=Constant3Vector() {constant=<rgba 1 0.1255 0.2902 1>}
   D=@n97
   Alpha=MaterialFunctionCall!(In=Add(A=@n141, B=@n130)) {materialFunction=&/MF_SmoothStepInverse}."InverseSmoothStep (Aprox)"
  ) {materialFunction=&/Lerp_4Color}
  Multiply(A=@n142.Result)
  Multiply(A=Add(A=@n136) {constB=-1100}) {constB=0.0015}
  n143 = Constant3Vector() {constant=<rgba 0 1 0 0>}
  n144 = Normalize(VectorInput=ComponentMask(None=Transform(None=@n143) {transformSourceType=TRANSFORMSOURCE_View, transformType=TRANSFORM_Tangent}) {g=true, r=true})
  n145 = Constant2Vector()
  Multiply(A=Add(A=TextureCoordinate(), B=Multiply(A=@n144, B=@n145)), B=Constant2Vector() {g=1, r=2}) {constA=1}
  Multiply(A=Add(A=@n136) {constB=-500}) {constB=0.0005}
  Multiply(A=Add(A=@n136) {constB=-1250}) {constB=0.001}
  n146 = DotProduct(A=MaterialFunctionCall() {materialFunction=&CameraDirectionVector}.Result, B=@n81)
  Multiply(A=Multiply(A=Sign(None=@n146)) {constB=-1}, B=@n31)
  n147 = OneMinus(None=TextureSample(UVs=Multiply(A=@n35) {constB=8}) {texture=&/T_Dots}.R)
  n148 = OneMinus(None=Multiply(A=@n132, B=Multiply(A=@n147) {constB=-0.4}))
  n149 = TextureSample(UVs=Multiply(
    A=MaterialFunctionCall!(Normal, Position) {materialFunction=&/MF_TriplanarUVs}.Triplanar UVs
    B=ScalarParameter() {defaultValue=0.0005, parameterName=tiling test}
   )) {texture=&T_Dots}
  n150 = Multiply(A=Multiply(A=OneMinus(None=@n149.R)) {constB=-0.2}, B=@n132)
  Multiply(A=@n148, B=OneMinus(None=@n150))
  Multiply(A=TextureCoordinate() {coordinateIndex=1}) {constB=4}
  Multiply(A=@n150)
  n151 = Saturate(None=Power(Base=LinearInterpolate(Alpha=@n130) {constA=-0.5}) {constExponent=3})
  n152 = Power(
   Base=LinearInterpolate(A=LinearInterpolate(A=@n142.Result, B=@n97, Alpha=@n151) {constB=0}, B=Constant3Vector() {constant=<rgba 0.9548 1 0.75 1>}, Alpha=@n131)
   Exp=LinearInterpolate(Alpha=@n130) {constA=3, constB=1.5}
  )
  n153 = Step(
   Y=TextureSample(UVs=@n58) {samplerType=SAMPLERTYPE_LinearGrayscale, texture=&/T_Bayer64_Grayscale_64px}.R
   X=LinearInterpolate(Alpha=@n64.Result) {constA=-0.2, constB=1.2}
  )
  n154 = Add(A=@n66) {constB=-0.5}
  n155 = ScalarParameter() {defaultValue=1, group=Delulu, parameterName=Dither stripes influence}
  n156 = Multiply(A=Saturate(None=Multiply(A=@n154, B=@n153)), B=OneMinus(None=Multiply(A=@n132, B=@n155)))
  n157 = LinearInterpolate(
   A=ScalarParameter() {defaultValue=1, group=Delulu, parameterName=Dither gradient offset}
   B=ScalarParameter() {defaultValue=0.3, group=Delulu, parameterName=Dither gradient offset top}
   Alpha=@n130
  )
  n158 = LinearInterpolate(A=Multiply(A=@n157, B=@n138), B=@n157, Alpha=Saturate(None=NamedRerouteUsage()))
  n159 = OneMinus(None=OneMinus(None=Multiply(A=@n156, B=@n158)))
  n160 = OneMinus(None=@n33.Result)
  n161 = Multiply(A=OneMinus(None=Multiply(A=@n159, B=@n160)), B=Multiply(A=@n148, B=@n152))
  n162 = MaterialFunctionCall(
   Color_Morning=Constant4Vector() {constant=<rgba 0.6 0.5 0.48 1>}.RGB
   Color_Day=Constant4Vector() {constant=<rgba 1 1 1 1>}.RGB
   Color_Evening=Constant4Vector() {constant=<rgba 0.53 0.48 0.6 1>}.RGB
   Color_Night=Constant4Vector() {constant=<rgba 0.3 0.3167 0.5 1>}.RGB
  ) {materialFunction=&Color_Time_of_Day}
  n163 = Constant() {r=1}
  n164 = Constant() {r=0.5}
  n165 = QualitySwitch(Default=@n163, Low=@n164, High=@n163, Medium=LinearInterpolate(A=@n163, B=@n164), Epic=@n163)
  NamedRerouteDeclaration(None=Multiply(A=Multiply(A=@n161, B=@n162.Result), B=@n165)) {name="(Delulu) Color Final", nodeColor=<rgba 0 1 0.8359 1>}
  n166 = NamedRerouteUsage()
  n167 = StaticSwitchParameter(True=Saturate(None=@n166), False=Saturate(None=Multiply(A=@n166) {constB=-1})) {defaultValue=true, group=Delulu, parameterName=Front Side}
  n168 = Constant() {r=1}
  n169 = LinearInterpolate(A=@n80, B=Multiply(A=@n80) {constB=2}, Alpha=@n130)
  NamedRerouteDeclaration(None=Multiply(
    A=LinearInterpolate(A=@n169, Alpha=@n33.Result) {desc=Distance fade}
    B=StaticSwitchParameter(True=@n167, False=@n168) {defaultValue=true, group=Delulu, parameterName=Single Sided}
   )) {name="(Delulu) Opacity", nodeColor=<rgba 0.5 1 0 1>}
  NamedRerouteDeclaration(None=Multiply(A=Multiply(A=@n152, B=@n162.Result), B=@n165)) {name="(Delulu) Color Spiral", nodeColor=<rgba 1 0 0.8203 1>}
  NamedRerouteDeclaration(None=@n162.Result) {name="(Delulu) Time of day recolor", nodeColor=<rgba 0.1719 0 1 1>}
  NamedRerouteDeclaration(None=StaticSwitchParameter(True=Constant() {r=1}, False=Constant() {r=-1}) {defaultValue=true, group=Delulu, parameterName=Front Side}) {name="(Delulu) TwoSidedSign", nodeColor=<rgba 1 0.6563 0 1>}
  n170 = Multiply(A=@n122) {constB=2}
  n171 = Frac(None=@n170) {desc=uv}
  n172 = Subtract(
   A=OneMinus(None=@n171)
   B=Multiply(A=MaterialFunctionCall(Value=@n171) {materialFunction=&LinearSine}.Linear Sine) {constB=0.5}
  )
  n173 = LinearInterpolate(Alpha=Ceil(None=Add(A=@n170) {constB=-1}) {desc=lmask}) {constA=0.1}
  n174 = Power(Base=@n119, Exp=@n173) {constExponent=0.5}
  NamedRerouteDeclaration(None=AppendVector(A=Multiply(A=@n172, B=@n174), B=@n119)) {name="(Delulu) Deploy", nodeColor=<rgba 0.5937 1 0 1>}
  NamedRerouteUsage()
  NamedRerouteUsage()
  Normalize(VectorInput=@n22)
  n175 = Add(
   A=MaterialFunctionCall() {materialFunction=&ComponentPivotLocation}.Result
   B=Constant3Vector()
  )
  n176 = CameraPositionWS()
  n177 = Add(A=Length(None=Subtract(A=@n175, B=@n176))) {constB=-500}
  n178 = MaterialFunctionCall(Alpha=Saturate(None=Multiply(A=@n177) {constB=0.0005})) {materialFunction=&SmoothStep}
  OneMinus(None=@n178.Result)
  n179 = TextureCoordinate()
  n180 = Frac(None=Add(A=ConstantBiasScale(None=@n179) {bias=-0.5, scale=2}, B=Constant2Vector() {g=0.5, r=0.5}))
  n181 = Constant() {r=0.5}
  OneMinus(None=Multiply(A=Distance(A=@n180, B=@n181)) {constB=2})
  n182 = Frac(None=Multiply(A=@n19) {constB=0.5})
  n183 = Constant() {r=0.5}
  OneMinus(None=Multiply(A=Distance(A=@n182, B=@n183)) {constB=2})
  n184 = Multiply(A=Reroute()) {constB=0.5}
  n185 = Multiply(A=Distance(A=Frac(None=@n184), B=Constant() {r=0.5})) {constB=2}
  OneMinus(None=@n185)
  OneMinus(None=@n85)
  Power(Exp=@n173)
  n186 = Saturate(None=Power(Base=Saturate(None=@n15), Exp=ScalarParameter() {defaultValue=5, group=HitWave, parameterName=Hitwave-pow}))
  n187 = ScalarParameter() {defaultValue=16, group=HitWave, parameterName=WPO-AddWave2}
  n188 = Multiply(A=VertexNormalWS(), B=NamedRerouteUsage())
  n189 = Constant()
  n190 = Constant()
  n191 = FeatureLevelSwitch(Default=@n190, ES2=@n189, ES3_1=@n189, SM4_REMOVED=@n190, SM5=@n190, SM6=@n190)
  n192 = Add(A=Multiply(A=Multiply(A=@n186, B=@n187), B=@n188), B=Multiply(A=@n188, B=QualitySwitch(Default=@n191, Low=@n189, High=@n190, Medium=@n189)))
  n193 = Multiply(A=VertexNormalWS(), B=NamedRerouteUsage())
  n194 = TextureSample() {samplerType=SAMPLERTYPE_Grayscale, texture=&LowResBlurredNoise}
  n195 = Constant() {r=1}
  n196 = ScalarParameter() {parameterName=Shrink}
  n197 = Multiply(
   A=QualitySwitch(
    Default=FeatureLevelSwitch(Default=@n194.R, ES2=@n195, ES3_1=@n195, SM4_REMOVED=@n194.R, SM5=@n194.R, SM6=@n194.R)
    Low=@n195
    High=@n194.R
    Medium=@n195
   )
   B=@n196
  )
  n198 = LinearInterpolate(A=Multiply(A=Multiply(A=@n193) {constB=-1}, B=LinearInterpolate(Alpha=@n197) {constB=3000}), Alpha=@n196) {constB=0}
  n199 = Multiply(A=Multiply(A=@n193) {constB=-1}, B=LinearInterpolate(Alpha=@n196) {constB=800})
  Reroute(None=Add(A=Add(A=@n192) {constB=0}, B=Add(A=@n198, B=@n199)) {constB=0})
  n200 = ComponentMask(None=Transform(None=PixelNormalWS()) {transformSourceType=TRANSFORMSOURCE_World, transformType=TRANSFORM_View}) {g=true, r=true}
  n201 = Add(A=Add(A=Multiply(A=@n200) {constB=-0.5}) {constB=0.5}, B=Constant2Vector() {g=0.5})
  n202 = ComponentMask(None=@n201) {g=true}
  n203 = Saturate(None=Add(A=LinearInterpolate(Alpha=@n202) {constA=0.5}) {constB=-0.65})
  n204 = OneMinus(None=Power(Base=OneMinus(None=@n203)) {constExponent=5})
  n205 = ComponentMask(None=@n201) {r=true}
  Reroute(None=Frac(None=AppendVector(A=@n205, B=@n204)))
  Reroute(None=@n178.Result)
  Reroute(None=@n128)
  Reroute(None=@n33.Result)
  Reroute(None=@n185)
  n206 = MaterialFunctionCall(Float2=TextureCoordinate()) {materialFunction=&BreakOutFloat2Components}
  n207 = Power(Base=Sine(None=Multiply(A=@n206.G) {constB=0.5}), Exp=ScalarParameter() {defaultValue=5, group=Grid, parameterName=GridFadeTopAndBotton})
  n208 = SphereMask(
   A=WorldPosition() {worldPositionShaderOffset=WPT_ExcludeAllShaderOffsets}.XYZ
   B=ComponentMask(None=VectorParameter() {group=HitWave, parameterName=HitWave-Loc, defaultValue=<rgba 100 -1400 -150 0>}.RGB) {b=true, g=true, r=true}
   Radius=Divide(A=ScalarParameter() {defaultValue=300, group=HitWave, parameterName=HitWave-SizeMultiplierOuterRing}) {constB=2.5}
  ) {attenuationRadius=128, hardnessPercent=0}
  n209 = TextureCoordinate() {uTiling=5, vTiling=5}
  n210 = Multiply(A=ComponentMask(None=VectorParameter() {group=Grid, parameterName=GridTiling, defaultValue=<rgba 3 3 0 0>}.RGB) {g=true, r=true}) {constB=2}
  n211 = Constant()
  n212 = QualitySwitch(Default=FeatureLevelSwitch(Default=@n16, ES2=@n211, ES3_1=@n211, SM4_REMOVED=@n16, SM5=@n16, SM6=@n16), Low=@n211, High=@n16, Medium=@n211)
  n213 = Constant()
  n214 = QualitySwitch(Default=FeatureLevelSwitch(Default=@n15, ES2=@n213, ES3_1=@n213, SM4_REMOVED=@n15, SM5=@n15, SM6=@n15), Low=@n213, High=@n15, Medium=@n213)
  n215 = Multiply(A=@n214, B=LinearInterpolate(B=ScalarParameter() {defaultValue=5, group=HitWave, parameterName=Hitwave-DistortAmt3}, Alpha=@n212))
  n216 = ScalarParameter() {defaultValue=1, parameterName=HitEffectLerp}
  n217 = TextureSampleParameter2D(UVs=Add(A=Multiply(A=@n209, B=@n210), B=Multiply(A=@n215, B=@n216))) {parameterName=GridTexture, samplerType=SAMPLERTYPE_LinearColor, texture=/Game/Gadgets/Assets/Teleporter/Textures/T_Teleporter_GridSquares_01}
  n218 = Saturate(None=Subtract(A=@n217.B) {constB=0.5})
  n219 = Time()
  n220 = ScalarParameter() {defaultValue=1, desc=Time-stamp, group=HitWave, parameterName=GameTime}
  n221 = Saturate(None=Divide(A=Subtract(A=@n219, B=@n220), B=ScalarParameter() {defaultValue=0.5, group=HitWave, parameterName=HitWave-DurationNew}))
  n222 = Multiply(
   A=LinearInterpolate(
    A=Multiply(A=@n217.R) {constB=4}
    B=Multiply(A=@n218) {constB=600, desc=Hit Grid Brightness}
    Alpha=ScalarParameter() {defaultValue=1, parameterName=HitEffectLerp}
   )
   B=Power(Base=OneMinus(None=@n221)) {constExponent=4}
  )
  n223 = PixelDepth()
  n224 = Constant() {r=-500}
  n225 = Saturate(None=Divide(A=Add(A=@n223.R, B=@n224), B=ScalarParameter() {defaultValue=6500, group=Opacity, parameterName=Distance_Fade}))
  n226 = MaterialFunctionCall!(
   TextureObject=TextureObjectParameter() {parameterName=Param, samplerType=SAMPLERTYPE_LinearColor, texture=&HighContrastNoise}
   TextureSize=VectorParameter() {group=Grid, parameterName=CausticWorldScale, defaultValue=<rgba 29400 36750 21000 0>}.RGB
   WorldPosition
   Export Float 4=StaticBool()
   In
   World Space Normal
   ProjectionTransitionContrast=Constant() {r=1}
  ) {materialFunction=/Game/Athena/SafeZone/WorldAlignedTexture_ScalePan}
  n227 = Multiply(A=Time() {desc=15.015014, period=1}, B=ScalarParameter() {defaultValue=0.08, desc=0.0666, group=Grid, parameterName=Sine Speed A_1})
  n228 = ComponentMask(None=Frac(None=Add(A=@n226.XY Texture, B=@n227))) {g=true}
  n229 = Constant() {r=0.5}
  n230 = OneMinus(None=Multiply(A=Distance(A=@n228, B=@n229)) {constB=2})
  n231 = Constant() {r=1}
  n232 = TextureCoordinate() {uTiling=5, vTiling=5}
  n233 = ComponentMask(None=VectorParameter() {group=Grid, parameterName=GridTiling, defaultValue=<rgba 3 3 0 0>}.RGB) {g=true, r=true}
  n234 = Multiply(A=@n214, B=LinearInterpolate(B=ScalarParameter() {defaultValue=5, group=HitWave, parameterName=Hitwave-DistortAmt3}, Alpha=@n212))
  n235 = ScalarParameter() {defaultValue=1, parameterName=HitEffectLerp}
  n236 = TextureSampleParameter2D(UVs=Add(A=Multiply(A=@n232, B=@n233), B=Multiply(A=@n234, B=@n235))) {parameterName=GridTexture, samplerType=SAMPLERTYPE_LinearColor, texture=&T_Teleporter_GridSquares_01}
  n237 = Time() {bOverride_Period=true, period=3.8314}
  n238 = ScalarParameter() {defaultValue=0.261, group=Grid, parameterName=Sine Speed Grid}
  n239 = Add(
   A=Multiply(
    A=FeatureLevelSwitch(Default=@n230, ES2=@n231, ES3_1=@n231, SM4_REMOVED=@n230)
    B=ScalarParameter() {defaultValue=0.2, group=Grid, parameterName=Ripple to Grid Scale}
   )
   B=Add(A=ComponentMask(None=@n236.RGB) {r=true}, B=Multiply(A=@n237, B=@n238))
  )
  n240 = Power(
   Base=Multiply(A=Sine(None=@n239), B=ComponentMask(None=@n236.RGB) {g=true})
   Exp=ScalarParameter() {defaultValue=5, group=Grid, parameterName=Grid Power}
  )
  n241 = Add(A=Multiply(A=@n207, B=Multiply(A=@n208, B=@n222)), B=Multiply(A=OneMinus(None=@n225), B=Multiply(A=@n240, B=@n207)))
  n242 = ScalarParameter() {defaultValue=1, group=Grid, parameterName=GridIntensity}
  Reroute(None=Multiply(A=Multiply(A=@n241, B=@n242), B=VectorParameter() {group=Grid, parameterName=Grid Color, defaultValue=<rgba 0.0143 0.063 0.15 0>}.RGB)) {desc=Grid}
  n243 = Panner(Coordinate=TextureCoordinate() {vTiling=0.8}) {speedY=0.4}
  n244 = Multiply(
   A=Multiply(A=VertexNormalWS(), B=NamedRerouteUsage())
   B=Multiply(A=TextureSample(UVs=@n243) {texture=/Game/Effects/Fort_Effects/Textures/Grad/gradient_healed}.R) {constB=12}
  )
  n245 = Constant()
  Reroute(None=QualitySwitch(
    Default=FeatureLevelSwitch(Default=@n244, ES2=@n245, ES3_1=@n245, SM4_REMOVED=@n244, SM5=@n244, SM6=@n244)
    Low=@n245
    High=@n244
    Medium=@n245
   ))
  n246 = TextureCoordinate()
  n247 = ScalarParameter() {defaultValue=0.0214, parameterName=LinesScale}
  n248 = Multiply(A=ScalarParameter() {defaultValue=5, group=HitWave, parameterName=Hitwave-DistortAmt3}) {constB=3}
  n249 = TextureSample(UVs=Add(A=Multiply(A=@n246, B=@n247), B=Multiply(A=@n214, B=@n248))) {samplerType=SAMPLERTYPE_LinearColor, texture=/Game/Athena/SafeZone/Lines}
  n250 = Constant()
  n251 = FeatureLevelSwitch(Default=@n249.R, ES2=@n250, ES3_1=@n250, SM4_REMOVED=@n249.R, SM5=@n249.R, SM6=@n249.R)
  Reroute(None=LinearInterpolate(Alpha=QualitySwitch(Default=@n251, Low=@n250, High=@n249.R, Medium=@n250)) {constA=0.2, constB=1.5})
  Reroute(None=Multiply(A=@n249.R) {constB=0.7})
  n252 = Divide(A=Time() {period=1}) {constB=7}
  n253 = Add(
   A=MaterialFunctionCall(UVs=TextureCoordinate() {coordinateIndex=1, vTiling=1.5}, Rotation Center=Constant2Vector() {g=0.5}, "Rotation Angle (0-1)"=@n252) {materialFunction=&CustomRotator}.Rotated Values
   B=@n252
  )
  n254 = Multiply(
   A=TextureSample(UVs=@n253) {samplerSource=SSM_Wrap_WorldGroupSettings, texture=&gradient_healed}.R
   B=TextureSample(UVs=@n253) {samplerSource=SSM_Wrap_WorldGroupSettings, texture=/Game/Effects/Fort_Effects/Textures/Mask/T_HalfSphere_D}.R
  )
  n255 = Divide(A=Time() {period=1}) {constB=-5}
  n256 = Add(
   A=MaterialFunctionCall(UVs=TextureCoordinate() {coordinateIndex=1, vTiling=1.2}, Rotation Center=Constant2Vector() {g=0.5, r=1}, "Rotation Angle (0-1)"=@n255) {materialFunction=&CustomRotator}.Rotated Values
   B=@n255
  )
  n257 = Multiply(
   A=TextureSample(UVs=@n256) {samplerSource=SSM_Wrap_WorldGroupSettings, texture=&gradient_healed}.R
   B=TextureSample(UVs=@n256) {samplerSource=SSM_Wrap_WorldGroupSettings, texture=&T_HalfSphere_D}.R
  )
  Reroute(None=Multiply(A=Add(A=@n254, B=@n257), B=VectorParameter() {group=Grid, parameterName=Grid Color, defaultValue=<rgba 0.0143 0.063 0.15 0>}.RGB))
  n258 = Multiply()
  n259 = Constant()
  Reroute(None=QualitySwitch(
    Default=FeatureLevelSwitch(Default=@n258, ES2=@n259, ES3_1=@n259, SM4_REMOVED=@n258, SM5=@n258, SM6=@n258)
    Low=@n259
    High=@n258
    Medium=@n259
   ))
  Reroute(None=@n32.Result)
  Reroute(None=@n217.R)
  n260 = VectorParameter() {group=Color, parameterName=ColorCenter, defaultValue=<rgba 0.0096 0.1133 0.1927 1>}
  n261 = Fresnel(
   ExponentIn=ScalarParameter() {defaultValue=3, group=Color, parameterName=FresnelExp}
   BaseReflectFractionIn=ScalarParameter() {defaultValue=0.0952, group=Color, parameterName=FresnelBase}
  )
  n262 = Multiply(
   A=VectorParameter() {group=Color, parameterName=ColorEdge, defaultValue=<rgba 0.0096 0.1133 0.1927 1>}.RGB
   B=Multiply(A=NamedRerouteUsage(), B=NamedRerouteUsage())
  )
  n263 = LinearInterpolate(
   A=LinearInterpolate(A=@n260.RGB, B=VectorParameter() {group=Color, parameterName=ColorMid, defaultValue=<rgba 0.0096 0.1133 0.1927 1>}.RGB, Alpha=@n261)
   B=@n262
   Alpha=Fresnel() {baseReflectFraction=0, exponent=10}
  )
  n264 = LinearInterpolate(A=@n260.RGB, B=Divide(A=@n262) {constB=4}, Alpha=@n261)
  n265 = FeatureLevelSwitch(Default=@n263, ES2=@n264, ES3_1=@n264, SM4_REMOVED=@n263, SM5=@n263, SM6=@n263)
  n266 = Add(
   A=Multiply(
    A=ScalarParameter() {defaultValue=1, group=Color, parameterName=FresnelBrightness}
    B=QualitySwitch(Default=@n265, Low=@n264, High=@n263, Medium=@n264)
   )
   B=@n49
  ) {desc=+ intersection edge}
  n267 = NamedRerouteUsage()
  n268 = ScalarParameter() {defaultValue=0.5, parameterName=HitEffect Photosensitivity reduction}
  n269 = Multiply(
   A=Multiply(A=VectorParameter() {group=HitWave, parameterName=HitWave-Tint, defaultValue=<rgba 0.8963 0.402 1 0>}.RGB, B=Multiply(A=@n268) {constB=0.5})
   B=@n15
  )
  n270 = ScalarParameter() {defaultValue=1, parameterName=HitEffectLerp}
  n271 = LinearInterpolate(B=Multiply(A=Multiply(B=@n269) {constA=1, constB=15}, B=LinearInterpolate(Alpha=@n270) {constB=85}) {constB=85}, Alpha=@n16)
  n272 = Constant()
  n273 = Add(
   A=QualitySwitch(
    Default=FeatureLevelSwitch(Default=@n271, ES2=@n272, ES3_1=@n272, SM4_REMOVED=@n271, SM5=@n271, SM6=@n271)
    Low=@n272
    High=@n271
    Medium=@n272
   )
   B=@n269
  )
  n274 = Multiply(
   A=@n268
   B=Multiply(
    A=Power(Base=@n16) {constExponent=4}
    B=VectorParameter() {group=HitWave, parameterName=Hitwave-AddBrighness, defaultValue=<rgba 0.0173 0.1326 0.2031 0>}.RGB
   )
  ) {constB=0}
  n275 = Constant()
  n276 = Saturate(None=Subtract(A=@n68, B=LinearInterpolate(Alpha=@n73) {constA=-0.2}))
  n277 = VectorParameter() {group=IntersectingEdge, parameterName=IntersectingEdgeColor, defaultValue=<rgba 0.0096 0.1133 0.1927 1>}
  n278 = NamedRerouteUsage()
  n279 = Add(A=Multiply(A=@n276) {constB=30}, B=LinearInterpolate(Alpha=Multiply(A=@n278, B=@n278)) {constA=-1, constB=3}) {constB=0}
  n280 = Multiply(A=MaterialFunctionCall(Float2=TextureCoordinate()) {materialFunction=&BreakOutFloat2Components}.G) {constB=0.5}
  n281 = Multiply(A=Multiply(A=@n276, B=Multiply(A=@n277.RGB, B=@n279)), B=Power(Base=Sine(None=@n280)) {constExponent=4})
  n282 = Add(
   A=Add(A=Add(A=@n266, B=@n267), B=Add(A=@n273, B=@n274))
   B=QualitySwitch(
    Default=FeatureLevelSwitch(Default=@n281, ES2=@n275, ES3_1=@n275, SM4_REMOVED=@n281, SM5=@n281, SM6=@n281)
    Low=@n275
    High=@n281
    Medium=@n275
   )
  )
  Reroute(None=MaterialFunctionCall(Emissive=Add(B=@n282)) {materialFunction=&MF_FX_EmissiveAdjustments}.Result)
  n283 = ComponentMask(None=VectorParameter() {group=Grid, parameterName=GridTiling, defaultValue=<rgba 3 3 0 0>}.RGB) {g=true, r=true}
  n284 = TextureSampleParameter2D(UVs=Multiply(A=TextureCoordinate(), B=Multiply(A=@n283) {constB=6})) {parameterName=Disolve_Testure, samplerType=SAMPLERTYPE_Masks, texture=/Game/Athena/Playlists/Barrier/Textures/T_BarrierGrid_M}
  n285 = Time()
  n286 = Frac(None=Add(A=@n284.R, B=Frac(None=@n285)))
  n287 = OneMinus(None=Multiply(A=Sine(None=@n286), B=@n284.G))
  n288 = ScalarParameter() {group=Opacity, parameterName=DissolveAmount}
  n289 = Multiply(A=ShadingPathSwitch(Default=@n287, Deferred=@n287, Forward=@n284.R, Mobile=@n284.R), B=@n288)
  n290 = Multiply(A=OneMinus(None=Saturate(None=@n289)), B=OneMinus(None=@n288))
  n291 = Constant() {r=1}
  n292 = FeatureLevelSwitch(Default=@n290, ES2=@n291, ES3_1=@n291, SM4_REMOVED=@n290, SM5=@n290, SM6=@n290)
  n293 = Sine(None=Multiply(A=@n2) {constB=0.5})
  n294 = Constant() {r=0.5}
  n295 = Power(Base=OneMinus(None=MaterialFunctionCall(Speed=@n294) {materialFunction=&DebugTimeSine}.Result)) {constExponent=1}
  n296 = LinearInterpolate(Alpha=StaticSwitchParameter(True=OneMinus(None=@n295), False=ScalarParameter() {parameterName=Deploy}) {parameterName=Debug}) {constA=-1, constB=0.2}
  n297 = LinearInterpolate(Alpha=@n69) {constA=0.8}
  n298 = Constant() {r=0.8}
  n299 = FeatureLevelSwitch(Default=@n297, ES2=@n298, ES3_1=@n298, SM4_REMOVED=@n297, SM5=@n297, SM6=@n297)
  n300 = Constant()
  n301 = Multiply(
   A=Multiply(A=@n54, B=QualitySwitch(Default=@n292, Low=@n291, High=@n290, Medium=@n291))
   B=If(
    A=Subtract(A=@n293, B=@n296)
    B=QualitySwitch(Default=@n299, Low=@n298, High=@n297, Medium=@n298)
    "A > B"=Constant() {r=1}
    "A == B"=@n300
    "A < B"=@n300
   )
  )
  n302 = DepthFade(Opacity=@n301, FadeDistance=ScalarParameter() {parameterName=Depthfade})
  n303 = Multiply(B=FeatureLevelSwitch(Default=@n302, ES2=@n301, ES3_1=@n301, SM4_REMOVED=@n302)) {constA=1}
  n304 = ScalarParameter() {defaultValue=1, parameterName=OverallOpacity}
  Reroute(None=Multiply(A=Multiply(A=@n303, B=@n304), B=NamedRerouteUsage()))
  n305 = OneMinus(None=MaterialFunctionCall(Speed=Constant() {r=0.5}) {materialFunction=&DebugTimeSine}.Result)
  StaticSwitchParameter(True=OneMinus(None=Power(Base=@n305) {constExponent=1}), False=@n196) {parameterName=Debug}
  Subtract(A=Frac(None=Add(A=@n109) {constB=0.5}), B=@n100)
  n306 = VectorParameter() {group=Grid, parameterName=GridTiling, defaultValue=<rgba 3 3 0 0>}
  TextureSample(UVs=Multiply(A=TextureCoordinate() {uTiling=5, vTiling=5}, B=ComponentMask(None=@n306.RGB) {g=true, r=true})) {samplerType=SAMPLERTYPE_LinearColor, texture=&T_Teleporter_GridSquares_01}
  TwoSidedSign()
  TwoSidedSign()
  TwoSidedSign()
  TwoSidedSign()
  TwoSidedSign()
  TwoSidedSign()
  TwoSidedSign()
  TwoSidedSign()
  TwoSidedSign()
  WorldPosition()
 ]
}
